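# Season UI item info widget, derived from FortItemInfoWidget, that shows an item's type, name and short description on a blurred, bordered panel, collapsed by default.
# The SetRarityMaterialValues event looks up rarity data from the given item definition.
# It writes the first two rarity colors into the name text's dynamic font material.
# The first color also tints the glow image behind the name through its dynamic material.
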WidgetBlueprint /Game/Athena/UI/Season/ItemInfoWidget
graphs [EventGraph {nodes [
   n1 = Event SetRarityMaterialValues!() -> (OutputDelegate, then, ItemDefinition:Item Definition Base Object Reference) {is=title K2Node_Event}
   > n2 = GetRarityData!(
    execute
    Rarity:EFortRarity Enum=GetRarity!(ItemDefinition:Item Definition Base Object Reference=@n1.ItemDefinition) -> (ReturnValue:EFortRarity Enum=Common)
   ) -> (then, ReturnValue:Fort Rarity Item Data Structure)
   > n3 = GetDynamicFontMaterial!(execute, self:Text Object Reference=Get Text_Name() -> (Text_Name:Common Text Object Reference)) -> (then, ReturnValue:Material Instance Dynamic Object Reference)
   n4 = Break Fort Rarity Item Data!(FortRarityItemData:"Fort Rarity Item Data Structure (by ref)"=@n2) -> (
    Name:Text
    Color1:Linear Color Structure
    Color2:Linear Color Structure
    Color3:Linear Color Structure
    Color4:Linear Color Structure
    Color5:Linear Color Structure
    Radius:float=~
    Falloff:float=~
    Brightness:float=~
    Roughness:float=~
    Glow:float=~
   ) {is=title K2Node_BreakStruct}
   > SetVectorParameterValue(self=@n3, ParameterName=MakeLiteralName(Value=RGB_1), Value=@n4.Color1)
   > SetVectorParameterValue(self=@n3, ParameterName=MakeLiteralName(Value=RGB_2), Value=@n4.Color2)
   > n5 = GetDynamicMaterial(self=Get RarityGlow() -> (RarityGlow:Image Object Reference))
   > SetVectorParameterValue(self=@n5, ParameterName=MakeLiteralName(Value=vRGB), Value=@n4.Color1)
  ]}]
variables []
widgets {
 info {inheritedWidgetCount=4, parentClass=/Script/FortniteUI.FortItemInfoWidget, rootWidgetClass=&Overlay, widgetCount=12}
 widgets [
  Overlay_MainView {widgetClassPath=&Overlay, bInherited=true, bIsVariable=true, widget=Overlay_MainView, properties {visibility=Collapsed}}
  BackgroundBlur_0 {
   widgetClassPath=&BackgroundBlur
   parent=Overlay_MainView
   widget=BackgroundBlur_0
   properties {blurStrength=8}
   slot {class=OverlaySlot, properties {horizontalAlignment=HAlign_Fill, verticalAlignment=VAlign_Fill}}
  }
  SizeBox_4 {
   widgetClassPath=&SizeBox
   parent=Overlay_MainView
   widget=SizeBox_4
   properties {bOverride_MinDesiredHeight=true, bOverride_WidthOverride=true, minDesiredHeight=256, widthOverride=448}
   slot {class=OverlaySlot}
  }
  CommonBorder_3 {
   widgetClassPath=&CommonBorder
   parent=SizeBox_4
   widget=CommonBorder_3
   properties {style=/FortUILibrary/Styles/Border/Border_Solid_DkBlue.Border_Solid_DkBlue_C, brushColor=<rgba 1 1 1 0.4>, padding=<ltrb 16 16 16 16>}
   slot {class=SizeBoxSlot}
  }
  MainViewItemInfoVB {
   widgetClassPath=&VerticalBox
   parent=CommonBorder_3
   widget=MainViewItemInfoVB
   slot {class=BorderSlot, properties {padding=<ltrb 16 16 16 16>}}
  }
  HorizontalBox_4 {widgetClassPath=&HorizontalBox, parent=MainViewItemInfoVB, widget=HorizontalBox_4, slot {class=VerticalBoxSlot}}
  Text_ItemType {
   widgetClassPath=&CommonTextBlock
   bInherited=true
   bIsVariable=true
   parent=HorizontalBox_4
   widget=Text_ItemType
   properties {
    mobileFontSizeMultiplier=1.34
    scrollStyle=&TextScrollStyle-Base_C
    style=/FortUILibrary/Styles/Typography/Foundation/TextStyle-Base-S-B-70pc.TextStyle-Base-S-B-70pc_C
    text=ITEM TYPE
    colorAndOpacity {colorUseRule=UseColor_Specified, specifiedColor=<rgba 1 1 1 0.7>}
    font {
     bForceMonospaced=false
     bMaterialIsStencil=false
     fontMaterial=null
     fontObject=&NotoSans
     letterSpacing=0
     monospacedWidth=1
     size=18
     skewAmount=0
     typefaceFontName=NotoSans-Bold
     outlineSettings {
      bApplyOutlineToDropShadows=false
      bMiteredCorners=false
      bSeparateFillAlpha=false
      outlineMaterial=null
      outlineSize=0
      outlineColor=<rgba 0 0 0 1>
     }
    }
    shadowOffset=<0 0>
   }
   slot {class=HorizontalBoxSlot, properties {size {sizeRule=Fill, value=1}}}
  }
  Overlay_1 {
   widgetClassPath=&Overlay
   parent=MainViewItemInfoVB
   widget=Overlay_1
   slot {class=VerticalBoxSlot, properties {horizontalAlignment=HAlign_Left}}
  }
  RarityGlow {
   widgetClassPath=&Image
   bIsVariable=true
   parent=Overlay_1
   widget=RarityGlow
   properties {
    brush {
     bIsDynamicallyLoaded=false
     drawAs=Box
     imageType=NoImage
     mirroring=NoMirror
     resourceName=None
     resourceObject=/FortUILibrary/Styles/Border/Materials/M_Border_RarityGlow
     tiling=NoTile
     imageSize=<96 8>
     margin=<ltrb 0.5 0 0.5 0>
     outlineSettings {
      bUseBrushTransparency=false
      roundingType=HalfHeightRadius
      width=0
      color {colorUseRule=UseColor_Specified, specifiedColor=<rgba 0 0 0 0>}
      cornerRadii=<0 0 0 0>
     }
     tintColor {colorUseRule=UseColor_Specified, specifiedColor=<rgba 1 1 1 1>}
     uVRegion {bIsValid=false, max=<0 0>, min=<0 0>}
    }
    colorAndOpacity=<rgba 1 1 1 0.25>
   }
   slot {class=OverlaySlot, properties {horizontalAlignment=HAlign_Fill, verticalAlignment=VAlign_Fill, padding=<ltrb -24 -4 -24 -2>}}
  }
  Text_Name {
   widgetClassPath=&CommonTextBlock
   bInherited=true
   bIsVariable=true
   parent=Overlay_1
   widget=Text_Name
   properties {
    mobileFontSizeMultiplier=1.34
    scrollStyle=&TextScrollStyle-Base_C
    style=/FortUILibrary/Styles/Typography/Foundation/TextStyle-Header-M-Horizon.TextStyle-Header-M-Horizon_C
    text=ITEM NAME
    font {
     bForceMonospaced=false
     bMaterialIsStencil=false
     fontMaterial=/FortUILibrary/Materials/Foundation/M_Font_Horizon_Dynamic
     fontObject=&Burbank
     letterSpacing=0
     monospacedWidth=1
     size=32
     skewAmount=0
     typefaceFontName=Burbank-Black
     outlineSettings {
      bApplyOutlineToDropShadows=true
      bMiteredCorners=false
      bSeparateFillAlpha=false
      outlineMaterial=null
      outlineSize=3
      outlineColor=<rgba 0 0 0 0.1>
     }
    }
    shadowOffset=<0 0>
   }
   slot {class=OverlaySlot, properties {horizontalAlignment=HAlign_Center, verticalAlignment=VAlign_Center}}
  }
  Rule {
   widgetClassPath=&Image
   parent=MainViewItemInfoVB
   widget=Rule
   properties {
    brush {
     bIsDynamicallyLoaded=false
     drawAs=Image
     imageType=NoImage
     mirroring=NoMirror
     resourceName=None
     resourceObject=null
     tiling=NoTile
     imageSize=<32 4>
     margin=<ltrb 0 0 0 0>
     outlineSettings {
      bUseBrushTransparency=false
      roundingType=HalfHeightRadius
      width=0
      color {colorUseRule=UseColor_Specified, specifiedColor=<rgba 0 0 0 0>}
      cornerRadii=<0 0 0 0>
     }
     tintColor {colorUseRule=UseColor_Specified, specifiedColor=<rgba 1 1 1 1>}
     uVRegion {bIsValid=false, max=<0 0>, min=<0 0>}
    }
    colorAndOpacity=<rgba 1 1 1 0.2>
   }
   slot {class=VerticalBoxSlot, properties {padding=<ltrb 0 4 0 8>}}
  }
  Text_Desc {
   widgetClassPath=&CommonTextBlock
   bInherited=true
   bIsVariable=true
   parent=MainViewItemInfoVB
   widget=Text_Desc
   properties {
    autoWrapText=true
    mobileFontSizeMultiplier=1.2
    style=/FortUILibrary/Styles/Typography/Foundation/TextStyle-Base-XS-40pc.TextStyle-Base-XS-40pc_C
    text=Short Description Text
    wrapTextAt=416
    colorAndOpacity {colorUseRule=UseColor_Specified, specifiedColor=<rgba 1 1 1 0.6>}
    font {
     bForceMonospaced=false
     bMaterialIsStencil=false
     fontMaterial=null
     fontObject=&NotoSans
     letterSpacing=0
     monospacedWidth=1
     size=16
     skewAmount=0
     typefaceFontName=NotoSans-Regular
     outlineSettings {
      bApplyOutlineToDropShadows=false
      bMiteredCorners=false
      bSeparateFillAlpha=false
      outlineMaterial=null
      outlineSize=0
      outlineColor=<rgba 0 0 0 1>
     }
    }
    shadowOffset=<0 0>
   }
   slot {class=VerticalBoxSlot}
  }
 ]
}
